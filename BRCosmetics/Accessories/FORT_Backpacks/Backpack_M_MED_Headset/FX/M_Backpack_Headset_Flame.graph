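# Unlit, two-sided translucent flame material for the Headset backpack effect, usable on Niagara sprites, ribbons and meshes.
# Two distorted panning noises and a sine-based shape mask drive opacity and curve-atlas color, with added plasma highlights and dissolve edge.
# World position offset combines optional spline thickening, noise and sine waves, and tip bending from dynamic parameter X alone.
# Dynamic parameters drive variation and dissolve; depth, view-angle and particle-centered sphere fades soften opacity.
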Material /BRCosmetics/Accessories/FORT_Backpacks/Backpack_M_MED_Headset/FX/M_Backpack_Headset_Flame
graph {
 properties {
  bUsedWithInstancedStaticMeshes=true
  bUsedWithNiagaraMeshParticles=true
  bUsedWithNiagaraRibbons=true
  bUsedWithNiagaraSprites=true
  blendMode=BLEND_Translucent
  pixelDepthOffsetMode=PDOM_Legacy
  shadingModel=MSM_Unlit
  twoSided=true
  usedShadingModels=Unlit
 }
 nodes [
  n1 = MaterialFunctionCall(
   WidthBase=ScalarParameter() {defaultValue=8, group=SplineThicken, parameterName=SplineWidthBase, sortPriority=2}
   WidthTip=ScalarParameter() {defaultValue=5, group=SplineThicken, parameterName=SplineWidthTip, sortPriority=1}
  ) {materialFunction=&SplineThicken}
  n2 = StaticSwitchParameter(
   True=DynamicParameter() {parameterIndex=3, paramNames [Dissolve, SplineWidthMul, Param3, Param4]}.SplineWidthMul
   False=ScalarParameter() {defaultValue=1, group=SplineThicken, parameterName=SplineWidthMul, sortPriority=3}
  ) {group=SplineThicken, parameterName=Use Dynamic SplineWidthMul}
  n3 = Constant()
  n4 = DynamicParameter() {desc=Parameters set by Niagara, parameterIndex=1, paramNames [X,Y,Z,WPOIntensity]}
  n5 = MaterialFunctionCall(X=@n4.X, Y=@n4.X, Z=@n4.X) {materialFunction=&MakeFloat3}
  n6 = ParticlePositionWS()
  n7 = ComponentMask(None=OneMinus(None=TextureCoordinate())) {g=true}
  n8 = ScalarParameter() {defaultValue=6, parameterName=WPOBendPower}
  n9 = LinearInterpolate(
   B=Clamp(
    None=Subtract(A=@n5.Result, B=@n6)
    Min=ScalarParameter() {defaultValue=-100, parameterName=Clamp_min}
    Max=ScalarParameter() {defaultValue=100, parameterName=Clamp_max}
   ) {maxDefault=1000, minDefault=-1000}
   Alpha=Multiply(A=Power(Base=@n7, Exp=@n8), B=@n4.WPOIntensity)
  )
  n10 = ScalarParameter() {defaultValue=1, parameterName=Use interpolate WPO}
  n11 = DynamicParameter() {defaultValue=<rgba 1 0 1 100>, paramNames [Random Offset, Noise Rotation, WPO Wave Intensity, Sphere Mask Range]}
  n12 = VectorParameter() {group=WPO, parameterName=WPO Side UV, sortPriority=6, defaultValue=<rgba 2 2 -0.1 0.5>}
  n13 = TextureCoordinate()
  n14 = Multiply(A=@n13, B=AppendVector(A=@n12.R, B=@n12.R))
  n15 = AppendVector(A=@n12.B, B=@n12.B)
  n16 = TextureSample(UVs=Add(A=AppendVector(A=@n11.Random Offset, B=@n11.Random Offset), B=Panner(Coordinate=@n14, Speed=@n15))) {samplerType=SAMPLERTYPE_Normal, texture=/Game/Characters/Player/Female/Medium/Bodies/F_MED_Bistro_Astronaut/FX/T_GaussianNoise_Normal}
  n17 = ComponentMask(None=@n16.RGB) {g=true, r=true}
  n18 = ScalarParameter() {defaultValue=2, group=WPO, parameterName=WPO Side Intensity}
  n19 = DynamicParameter() {defaultValue=<rgba 1 0 1 100>, paramNames [Random Offset, Noise Rotation, WPO Wave Intensity, Sphere Mask Range]}
  n20 = AppendVector(A=@n19.Random Offset, B=@n19.Random Offset)
  n21 = VectorParameter() {group=WPO, parameterName=WPO Wave UV, sortPriority=6, defaultValue=<rgba 2 2 0 1.5>}
  n22 = Panner(Coordinate=Multiply(A=@n13, B=AppendVector(A=@n21.R, B=@n21.R)), Speed=AppendVector(A=@n21.B, B=@n21.B))
  n23 = Multiply(
   A=TextureSample(UVs=Add(A=@n20, B=@n22)) {texture=&T_Smooth_tile}.R
   B=ScalarParameter() {defaultValue=1, group=WPO, parameterName=WPO Wave Intensity}
  ) {constB=10}
  n24 = DynamicParameter() {defaultValue=<rgba 1 0 1 100>, paramNames [Random Offset, Noise Rotation, WPO Wave Intensity, Sphere Mask Range]}
  n25 = TextureCoordinate()
  n26 = ComponentMask(None=Add(A=@n24.Random Offset, B=Panner(Coordinate=@n25) {speedY=0.5})) {g=true}
  n27 = Saturate(None=Power(Base=Sine(None=@n26)) {constExponent=40})
  n28 = ScalarParameter() {defaultValue=-4, group=WPO, parameterName=Sine Wave Intensity}
  n29 = TextureCoordinate()
  n30 = StaticSwitchParameter(True=Multiply(A=Multiply(A=@n27, B=@n28) {constB=-4}, B=ComponentMask(None=@n29) {g=true}), False=Constant()) {group=WPO, parameterName=Use Sine Wave}
  n31 = TextureCoordinate()
  n32 = LinearInterpolate(B=AppendVector(A=Multiply(A=@n17, B=@n18) {constB=10}, B=Add(A=@n23, B=@n30)), Alpha=OneMinus(None=ComponentMask(None=@n31) {g=true}))
  n33 = DynamicParameter() {defaultValue=<rgba 1 0 1 100>, paramNames [Random Offset, Noise Rotation, WPO Wave Intensity, Sphere Mask Range]}
  Add(
   A=StaticSwitchParameter(True=Multiply(A=@n1.WorldPositionOffset, B=@n2), False=Constant()) {group=SplineThicken, parameterName=UseSplineThicken, sortPriority=0}
   B=Add(A=LinearInterpolate(A=@n3, B=@n9, Alpha=@n10), B=Multiply(A=@n32, B=@n33.WPO Wave Intensity) {constB=10})
  )
  n34 = TextureCoordinate()
  n35 = VectorParameter() {group=Emissive, parameterName=Highlight UV, sortPriority=6, defaultValue=<rgba 4 1 -1 1>}
  n36 = AppendVector(A=@n35.R, B=@n35.R)
  n37 = Multiply(
   A=Panner(Coordinate=Multiply(A=@n34, B=@n36), Speed=AppendVector(A=@n35.B, B=@n35.B))
   B=AppendVector(
    A=ScalarParameter() {defaultValue=1, group=Noise, parameterName=Global Scale U}
    B=ScalarParameter() {defaultValue=1, group=Noise, parameterName=Global Scale V}
   )
  )
  n38 = Constant() {r=1}
  n39 = Desaturation(
   None=Power(Base=TextureSample!(UVs=@n37, Tex, Bias=@n38, Apply View MipBias) {mipValueMode=TMVM_MipBias, texture=/Game/Effects/Fort_Effects/Textures/Plasma/T_Plasma_Aurora_Seamless}.RGB)
   Fraction=ScalarParameter() {defaultValue=0.5, group=Emissive, parameterName=HightLight Color Desaturation}
  )
  n40 = VectorParameter() {group=Emissive, parameterName=Highlight Color, defaultValue=<rgba 0.0333 0 0.2 1>}
  n41 = DynamicParameter() {defaultValue=<rgba 1 0 1 100>, paramNames [Random Offset, Noise Rotation, WPO Wave Intensity, Sphere Mask Range]}
  n42 = AppendVector(A=@n41.Random Offset, B=@n41.Random Offset)
  n43 = AppendVector(
   A=ScalarParameter() {defaultValue=1, group=Noise, parameterName=Global Scale U}
   B=ScalarParameter() {defaultValue=1, group=Noise, parameterName=Global Scale V}
  )
  n44 = VectorParameter() {group=Noise, parameterName=Noise B UV, sortPriority=6, defaultValue=<rgba 2 0.5 -1 0>}
  n45 = AppendVector(A=@n44.R, B=@n44.R)
  n46 = MaterialFunctionCall(UVs=TextureCoordinate(), "Rotation Angle (0-1)"=@n41.Noise Rotation) {materialFunction=&CustomRotator}
  n47 = Panner(Coordinate=Multiply(A=@n46.Rotated Values, B=Multiply(A=@n43, B=@n45)), Speed=AppendVector(A=@n44.B, B=@n44.B))
  n48 = VectorParameter() {group=Distortion, parameterName=Distortion UV, sortPriority=1, defaultValue=<rgba 1 1 0 0>}
  n49 = ScalarParameter() {defaultValue=1, group=Noise, parameterName=Global Scale U}
  n50 = ScalarParameter() {defaultValue=1, group=Noise, parameterName=Global Scale V}
  n51 = Multiply(A=@n46.Rotated Values, B=Multiply(A=AppendVector(A=@n48.R, B=@n48.R), B=AppendVector(A=@n49, B=@n50)))
  n52 = AppendVector(A=@n48.B, B=@n48.B)
  n53 = TextureSampleParameter2D(UVs=Add(A=AppendVector(A=@n41.Random Offset, B=@n41.Random Offset), B=Panner(Coordinate=@n51, Speed=@n52))) {
   group=Distortion
   parameterName="Distortion  Tex"
   samplerType=SAMPLERTYPE_Normal
   sortPriority=0
   texture=&T_Weap_Distort_N_02
  }
  n54 = OneMinus(None=ComponentMask(None=@n46.Rotated Values) {g=true})
  n55 = ScalarParameter() {defaultValue=0.5, group=Distortion, parameterName=Distortion Gradient Lerp, sortPriority=3}
  n56 = Multiply(
   A=Multiply(A=ComponentMask(None=@n53.RGB) {g=true, r=true}, B=LinearInterpolate(B=@n54, Alpha=@n55) {constA=1, constAlpha=0})
   B=ScalarParameter() {defaultValue=0.1, group=Distortion, parameterName=Distortion Intensity, sortPriority=2}
  )
  n57 = TextureSampleParameter2D(UVs=Add(A=Add(A=@n42, B=@n47), B=@n56)) {
   group=Noise
   parameterName=NoIse B Texture
   sortPriority=4
   texture=/BRCosmetics/Weapons/FORT_Melee/Pickaxe_Headset/FX/T_Pickaxe_Headset_Noise_01
  }
  n58 = ChannelMaskParameter(None=@n57.RGB) {group=Noise, maskChannel=Green, parameterName=Noise B Channel, sortPriority=5, defaultValue=<rgba 0 1 0 0>}
  n59 = ScalarParameter() {defaultValue=1, group=Noise, parameterName=NoIse B Intensity, sortPriority=7}
  n60 = AppendVector(A=@n41.Random Offset, B=@n41.Random Offset)
  n61 = VectorParameter() {group=Noise, parameterName=Noise A UV, sortPriority=2, defaultValue=<rgba 2 0.5 -1.5 0>}
  n62 = AppendVector(A=@n61.R, B=@n61.R)
  n63 = Panner(Coordinate=Multiply(A=@n46.Rotated Values, B=Multiply(A=@n43, B=@n62)), Speed=AppendVector(A=@n61.B, B=@n61.B))
  n64 = TextureSampleParameter2D(UVs=Add(A=@n56, B=Add(A=@n60, B=@n63))) {
   group=Noise
   parameterName=NoIse A Texture
   sortPriority=0
   texture=/Game/Global/Textures/Default/Tileable_Patterns/T_Blowing_Smoke_Pattern2_Tile_DISP
  }
  n65 = ChannelMaskParameter(None=@n64.RGB) {group=Noise, parameterName=Noise A Channel, sortPriority=1}
  n66 = ScalarParameter() {defaultValue=0.5, group=Noise, parameterName=NoIse A Intensity, sortPriority=3}
  n67 = LinearInterpolate(
   A=Saturate(None=Multiply(A=@n58.R, B=@n59) {constB=10})
   B=Saturate(None=Multiply(A=@n65.R, B=@n66) {constB=10})
   Alpha=ScalarParameter() {defaultValue=0.5, group=Noise, parameterName=Lerp Noise AB, sortPriority=8}
  )
  n68 = MaterialFunctionCall(Float2=TextureCoordinate()) {materialFunction=&BreakOutFloat2Components}
  n69 = Multiply(A=@n68.R) {constB=0.5}
  n70 = ScalarParameter() {defaultValue=2, group=Shape, parameterName=Top Thickness, sortPriority=0}
  n71 = ScalarParameter() {group=Shape, parameterName=Bottom Thickness, sortPriority=1}
  n72 = Multiply(A=Power(Base=Sine(None=@n69), Exp=LinearInterpolate(A=@n70, B=@n71, Alpha=@n68.G) {constA=6, constB=2}) {constExponent=4}, B=@n68.G)
  n73 = Add(A=@n67, B=Saturate(None=Power(Base=@n72)))
  n74 = ScalarParameter() {defaultValue=1, group=Shape, parameterName=Top Fade, sortPriority=2}
  n75 = ScalarParameter() {defaultValue=1, group=Shape, parameterName=Bottom Fade, sortPriority=3}
  n76 = CurveAtlasRowParameter(CurveTime=ComponentMask(None=TextureCoordinate()) {g=true}) {
   atlas=/Game/Athena/LinearColorCurves/A_Atlas_Turtleneck_Glider
   curve=/Game/Accessories/FORT_Backpacks/Backpack_M_MED_Headset/FX/Curve_Backpack_Headset_Flame
   defaultValue=32
   group=Emissive
   parameterName=ColorCurve
   sortPriority=0
  }
  n77 = TextureCoordinate()
  n78 = VectorParameter() {
   group=BottomFade
   parameterName=Bottom Fade_SphereMask_Scale
   sortPriority=1
   channelNames {a=Center V, b=Center U, g=Scale V, r=Scale U}
   defaultValue=<rgba 1 1 0.5 0>
  }
  n79 = AppendVector(A=@n78.Scale U, B=@n78.Scale U)
  n80 = Power(
   Base=SphereMask(
    A=Multiply(A=@n77, B=@n79)
    B=AppendVector(A=@n78.Center U, B=@n78.Center U)
    Radius=ScalarParameter() {defaultValue=0.95, group=BottomFade, parameterName=BottomFade_SphereMask_Radius, sortPriority=2}
   ) {attenuationRadius=0.95, hardnessPercent=0}
   Exp=ScalarParameter() {defaultValue=2, group=BottomFade, parameterName=BottomFade_SphereMask_Power, sortPriority=3}
  )
  n81 = Multiply(
   A=Power(Base=Multiply(A=@n73, B=@n72), Exp=LinearInterpolate(A=@n74, B=@n75, Alpha=@n68.G) {constA=1, constB=0.5})
   B=LinearInterpolate(A=@n76.A, B=Saturate(None=@n80), Alpha=ScalarParameter() {group=BottomFade, parameterName=Use Bottom Fade, sortPriority=0}) {constA=1}
  )
  n82 = AppendVector(
   A=ScalarParameter() {defaultValue=1, group=Noise, parameterName=Global Scale U}
   B=ScalarParameter() {defaultValue=1, group=Noise, parameterName=Global Scale V}
  )
  n83 = TextureCoordinate()
  n84 = VectorParameter() {group=Noise, parameterName=DarkDetail UV, sortPriority=6, defaultValue=<rgba 4 0.5 0 1>}
  n85 = Panner(Coordinate=Multiply(A=Multiply(A=@n82, B=@n83), B=AppendVector(A=@n84.R, B=@n84.R)), Speed=AppendVector(A=@n84.B, B=@n84.B))
  n86 = Multiply(
   A=Constant() {r=0.1}
   B=Power(Base=TextureSample!(UVs=@n85, Tex, Apply View MipBias) {texture=&T_Inky_Smoke_Tile}.R) {constExponent=1}
  )
  n87 = CurveAtlasRowParameter(CurveTime=Subtract(A=@n81, B=Saturate(None=@n86))) {
   atlas=&A_Atlas_Turtleneck_Glider
   curve=&Curve_Backpack_Headset_Flame
   defaultValue=32
   group=Emissive
   parameterName=ColorCurve
   sortPriority=0
  }
  n88 = Add(
   A=Multiply(A=Multiply(A=@n39, B=@n40.RGB)) {constB=10}
   B=Desaturation(
    None=@n87
    Fraction=DynamicParameter() {parameterIndex=2, defaultValue=<rgba 0 1 1 1>, paramNames [Color Desaturation, "", "", ""]}.Color Desaturation
   )
  )
  n89 = ScalarParameter() {defaultValue=0.75, group=Emissive, parameterName=Emissive_Intensity}
  n90 = VectorParameter() {group=Dissolve, parameterName=Dissolve Color, defaultValue=<rgba 400.0013 0 600 1>}
  n91 = Saturate(None=@n67)
  n92 = ComponentMask(None=TextureCoordinate()) {g=true}
  n93 = Power(
   Base=LinearInterpolate(A=Multiply(A=@n91, B=@n92), B=@n92, Alpha=Constant() {r=0.4})
   Exp=ScalarParameter() {defaultValue=0.5, group=Dissolve, parameterName=Dissolve Power}
  )
  n94 = ScalarParameter() {defaultValue=30, group=Dissolve, parameterName=Dissolve Edge Hardness, sliderMin=0.5}
  n95 = DynamicParameter() {parameterIndex=3, paramNames [Dissolve, SplineWidthMul, Param3, Param4]}
  n96 = Saturate(None=Subtract(A=Multiply(A=@n93, B=@n94), B=LinearInterpolate(A=@n94, Alpha=@n95.Dissolve) {constB=-1}))
  n97 = Subtract(A=Constant() {r=1}, B=@n96)
  n98 = ParticleColor()
  n99 = Multiply(A=Add(A=Multiply(A=@n88, B=@n89), B=Multiply(A=@n90.RGB, B=@n97)), B=@n98.RGB)
  MaterialFunctionCall(Emissive=@n99) {materialFunction=&MF_FX_EmissiveAdjustments}
  n100 = Multiply(A=Multiply(A=@n81, B=@n96), B=ScalarParameter() {defaultValue=20, group=Opacity, parameterName=Opacity Multiplier, sortPriority=1}) {constB=3}
  n101 = ScalarParameter() {defaultValue=20, group=Opacity, parameterName=DepthFadeDistance}
  n102 = Multiply(A=@n98.A, B=ShadingPathSwitch(Default=DepthFade(Opacity=@n100, FadeDistance=@n101), Mobile=@n100) {desc=Migrated from FeatureLevelSwitch})
  n103 = MaterialFunctionCall("FadeSensitivity (optional)"=ScalarParameter() {defaultValue=0.5, group=Opacity, parameterName=FadeBasedOnViewAngle}) {materialFunction=&FadeBasedOnViewAngle}
  n104 = WorldPosition()
  n105 = ParticlePositionWS()
  n106 = DynamicParameter() {defaultValue=<rgba 1 0 1 100>, paramNames [Random Offset, Noise Rotation, WPO Wave Intensity, Sphere Mask Range]}
  Reroute(None=Multiply(
    A=Multiply(A=@n102, B=@n103.Result)
    B=SphereMask(A=@n104.XYZ, B=@n105, Radius=@n106.Sphere Mask Range) {attenuationRadius=50, hardnessPercent=0.2}
   ))
  Reroute(None=@n99)
  ScalarParameter() {defaultValue=0.5, group=Dissolve, parameterName=Dissolve Dynamic Test}
 ]
}
