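# Translucent unlit Niagara mesh-particle material for the headset backpack cosmetic, rendering glowing, dissolving noise.
# Two panning noise textures, offset by a scrolling distortion normal map, blend and add to a UV V-gradient to mix two colors.
# A particle dynamic parameter drives a noise-modulated dissolve along V, adding a hot edge color to emissive.
# Opacity combines the dissolve mask, a V-gradient fade, Fresnel and particle alpha.
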Material /BRCosmetics/Accessories/FORT_Backpacks/Backpack_M_MED_Headset/FX/M_Backpack_Headset_MeshNoise
graph {
 properties {
  bUsedWithNiagaraMeshParticles=true
  blendMode=BLEND_Translucent
  pixelDepthOffsetMode=PDOM_Legacy
  previewMesh=&/SM_Backpack_Headset_Horn_L
  shadingModel=MSM_Unlit
  usedShadingModels=Unlit
 }
 nodes [
  n1 = TextureCoordinate()
  n2 = Power(
   Base=OneMinus(None=ComponentMask(None=@n1) {g=true})
   Exp=ScalarParameter() {defaultValue=10, group=Gradient, parameterName=Emissive Gradient Power}
  ) {constExponent=10}
  n3 = ScalarParameter() {defaultValue=10, group=Gradient, parameterName=Emissive Gradient Mult}
  n4 = TextureCoordinate()
  n5 = VectorParameter() {parameterName=Distortion UV, sortPriority=2, defaultValue=<rgba 1 1 0 0>}
  n6 = AppendVector(A=@n5.R, B=@n5.R)
  n7 = TextureSampleParameter2D(UVs=Panner(Coordinate=Multiply(A=@n4, B=@n6), Speed=AppendVector(A=@n5.B, B=@n5.B))) {
   parameterName=Distortion Tex
   samplerType=SAMPLERTYPE_Normal
   sortPriority=0
   texture=&T_Weap_Distort_N_02
  }
  n8 = Multiply(A=ComponentMask(None=@n7.RGB) {g=true, r=true}, B=ScalarParameter() {defaultValue=0.1, parameterName=Distortion Intensity, sortPriority=2})
  n9 = TextureCoordinate()
  n10 = VectorParameter() {parameterName=Noise A UV, sortPriority=2, defaultValue=<rgba 1 0.75 -0.2 0.5>}
  n11 = AppendVector(A=@n10.R, B=@n10.R)
  n12 = Add(A=@n8, B=Panner(Coordinate=Multiply(A=@n9, B=@n11), Speed=AppendVector(A=@n10.B, B=@n10.B)))
  n13 = TextureCoordinate()
  n14 = VectorParameter() {parameterName=Noise B UV, sortPriority=2, defaultValue=<rgba 1 0.5 -0.1 0.3>}
  n15 = AppendVector(A=@n14.R, B=@n14.R)
  n16 = Add(A=@n8, B=Panner(Coordinate=Multiply(A=@n13, B=@n15), Speed=AppendVector(A=@n14.B, B=@n14.B)))
  n17 = Add(
   A=Saturate(None=Multiply(A=@n2, B=@n3) {constB=10})
   B=LinearInterpolate(
    A=TextureSample(UVs=@n12) {texture=/Game/Global/Textures/Default/Tileable_Patterns/T_Blowing_Smoke_Pattern2_Tile_DISP}.R
    B=TextureSample(UVs=@n16) {samplerType=SAMPLERTYPE_Grayscale, texture=&LowResBlurredNoise}.R
   )
  )
  n18 = LinearInterpolate(
   A=VectorParameter() {parameterName=Color B, defaultValue=<rgba 0.0083 0 0.05 1>}.RGB
   B=VectorParameter() {parameterName=Color A, defaultValue=<rgba 1.6667 0 5 1>}.RGB
   Alpha=@n17
  )
  n19 = Saturate(None=@n17)
  n20 = ComponentMask(None=TextureCoordinate()) {g=true}
  n21 = Power(
   Base=LinearInterpolate(A=Multiply(A=@n19, B=@n20), B=@n20, Alpha=Constant() {r=0.4})
   Exp=ScalarParameter() {defaultValue=0.5, group=Dissolve, parameterName=Dissolve Power}
  )
  n22 = ScalarParameter() {defaultValue=30, group=Dissolve, parameterName=Dissolve Edge Hardness, sliderMin=0.5}
  n23 = DynamicParameter() {paramNames [Dissolve,Param2,Param3,Param4]}
  n24 = Saturate(None=Subtract(A=Multiply(A=@n21, B=@n22), B=LinearInterpolate(A=@n22, Alpha=@n23.Dissolve) {constB=-1}))
  n25 = Multiply(A=VectorParameter() {parameterName=Dissolve Color, defaultValue=<rgba 333.334 0 500 1>}.RGB, B=Subtract(A=Constant() {r=1}, B=@n24))
  n26 = ParticleColor()
  MaterialFunctionCall(Emissive=Multiply(A=Add(A=@n18, B=@n25), B=@n26.RGB)) {materialFunction=&MF_FX_EmissiveAdjustments}
  n27 = OneMinus(None=ComponentMask(None=TextureCoordinate()) {g=true})
  n28 = ScalarParameter() {defaultValue=4, group=Gradient, parameterName=Alpha Gradient Power}
  n29 = Saturate(None=Multiply(A=Power(Base=@n27, Exp=@n28) {constExponent=4}, B=ScalarParameter() {defaultValue=20, group=Gradient, parameterName=Alpha Gradient Mult}) {constB=20})
  n30 = Fresnel(
   ExponentIn=ScalarParameter() {defaultValue=1, group=Fresnel, parameterName=Fresnel ExponentIn}
   BaseReflectFractionIn=ScalarParameter() {defaultValue=0.2, group=Fresnel, parameterName=Fresnel BaseReflectFractionIn}
  ) {baseReflectFraction=0.2, exponent=1}
  Multiply(A=@n26.A, B=Multiply(A=@n24, B=Multiply(A=@n29, B=@n30)))
  ScalarParameter() {defaultValue=0.5, group=Dissolve, parameterName=Dissolve Dynamic Test}
 ]
}
